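# Sound cue for the drop sound of the AppleSauce air strike weapon in the Figment_AirStrike content.
# A single wave player feeds the output directly with the AppleSauce_Drop_01 wave, with no randomization or mixing.
# It uses the shared weapon drop attenuation and concurrency set, and routes through the loot collect sound class.
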SoundCue /Figment_AirStrike/Audio/SC_Figment_AppleSauce_Drop_Cue
root=WavePlayer_1
nodes [{id=WavePlayer_1, properties {soundWaveAssetPtr=/Game/Athena/Sounds/Weapons/Air_Strike/AppleSauce_Drop_01}}]
properties {
 attenuationSettings=&Fort_Weapon_Drop_Att
 maxDistance=3000
 soundClassObject=&Fort_Loot_Collect
 volumeMultiplier=1
 concurrencyOverrides {resolutionRule=PreventNew}
 concurrencySet [&Weapon_Drop_StopOldest]
}
