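# Sound cue for the Figment air strike (AppleSauce) explosion, layered by distance.
# A mixer blends close, medium and distant branches, each with randomized waves, pitch modulation, its own attenuation and a sound class override.
# The close layer uses dedicated AppleSauce waves, while medium and distant layers reuse Quad Launcher and RPG explosion waves.
# Concurrency overrides cap simultaneous instances, stopping the oldest and scaling down volume.
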SoundCue /Figment_AirStrike/Audio/SC_Figment_AppleSauce_Explosion_Cue
root=Mixer_0
nodes [
 {
  id=Attenuation_0
  children [Random_0]
  properties {attenuationSettings=&/ATT_Figment_AppleSauce_Explosion_Close, attenuationOverrides {falloffDistance=1600}}
 }
 {
  id=Attenuation_1
  children [Modulator_1]
  properties {
   attenuationSettings=&/ATT_Figment_AppleSauce_Explosion_Medium
   attenuationOverrides {falloffDistance=3000, attenuationShapeExtents {x=1000}}
  }
 }
 {
  id=Attenuation_2
  children [Modulator_3]
  properties {
   attenuationSettings=&/ATT_Figment_AppleSauce_Explosion_Distant
   attenuationOverrides {falloffDistance=3000, attenuationShapeExtents {x=1000}}
  }
 }
 {id=Mixer_0, children [SoundClass_1,SoundClass_0,SoundClass_4], properties {inputVolume [1,1,1]}}
 {id=Modulator_1, children [Random_3], properties {pitchMax=1.1, pitchMin=0.9}}
 {id=Modulator_2, children [Attenuation_0], properties {pitchMax=1.2}}
 {id=Modulator_3, children [Random_2], properties {pitchMax=1.1, pitchMin=0.9}}
 {id=Random_0, children [WavePlayer_8, WavePlayer_12, WavePlayer_13, WavePlayer_14], properties {weights [1,1,1,1]}}
 {id=Random_2, children [WavePlayer_9, WavePlayer_10, WavePlayer_11], properties {weights [1,1,1]}}
 {id=Random_3, children [WavePlayer_1,WavePlayer_6,WavePlayer_7], properties {weights [1,1,1]}}
 {
  id=SoundClass_0
  children [Attenuation_1]
  properties {soundClassOverride=&Fort_WeaponRanged_Fire_Distant}
 }
 {id=SoundClass_1, children [Modulator_2], properties {soundClassOverride=&Fort_WeaponRanged_Fire_3P_PSM}}
 {
  id=SoundClass_4
  children [Attenuation_2]
  properties {soundClassOverride=&Fort_WeaponRanged_Fire_Distant}
 }
 {id=WavePlayer_1, properties {soundWaveAssetPtr=/Game/Athena/Sounds/Weapons/Explosions/Quad_Launcher_Expl_Med_03}}
 {id=WavePlayer_10, properties {soundWaveAssetPtr=/Game/Athena/Sounds/Weapons/Explosions/RPG_Expl_Dist_02}}
 {id=WavePlayer_11, properties {soundWaveAssetPtr=/Game/Athena/Sounds/Weapons/Explosions/RPG_Expl_Dist_03}}
 {id=WavePlayer_12, properties {soundWaveAssetPtr=/Game/Athena/Sounds/Weapons/Air_Strike/AppleSauce_Expl_Close_01}}
 {id=WavePlayer_13, properties {soundWaveAssetPtr=/Game/Athena/Sounds/Weapons/Air_Strike/AppleSauce_Expl_Close_02}}
 {id=WavePlayer_14, properties {soundWaveAssetPtr=/Game/Athena/Sounds/Weapons/Air_Strike/AppleSauce_Expl_Close_03}}
 {id=WavePlayer_6, properties {soundWaveAssetPtr=/Game/Athena/Sounds/Weapons/Explosions/Quad_Launcher_Expl_Med_01}}
 {id=WavePlayer_7, properties {soundWaveAssetPtr=/Game/Athena/Sounds/Weapons/Explosions/Quad_Launcher_Expl_Med_02}}
 {id=WavePlayer_8, properties {soundWaveAssetPtr=/Game/Athena/Sounds/Weapons/Air_Strike/AppleSauce_Expl_Close_04}}
 {id=WavePlayer_9, properties {soundWaveAssetPtr=/Game/Athena/Sounds/Weapons/Explosions/RPG_Expl_Dist_01}}
]
properties {
 bOverrideConcurrency=true
 maxDistance=18000
 volumeMultiplier=0.8
 concurrencyOverrides {maxCount=5, resolutionRule=StopOldest, volumeScale=0.5, volumeScaleAttackTime=0.1}
 concurrencySet [/Game/Sounds/Concurrency_Rules/Weapon_Explosion_Conc]
}
