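# Sound cue for the Figment AirStrike AppleSauce explosion on water, per its name, built from stock grenade and distant explosion waves.
# A mixer blends a quieter medium-range layer with a boosted distant layer.
# Each layer has its own attenuation, modulator and sound class override (third-person weapon fire versus distant fire).
# The cue references the shared weapon explosion concurrency set to limit stacking.
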SoundCue /Figment_AirStrike/Audio/SC_Figment_AppleSauce_Explosion_Water_Cue
root=Mixer_4
nodes [
 {
  id=Attenuation_4
  children [WavePlayer_0]
  properties {attenuationSettings=&/ATT_Figment_AppleSauce_Rocket_Medium_Explosion, attenuationOverrides {falloffDistance=1600}}
 }
 {
  id=Attenuation_5
  children [Modulator_5]
  properties {
   attenuationSettings=&/ATT_Figment_AppleSauce_Rocket_Distant_Explosion
   attenuationOverrides {falloffDistance=3000, attenuationShapeExtents {x=1000}}
  }
 }
 {id=Mixer_4, children [SoundClass_3,SoundClass_2], properties {inputVolume [0.5,1.2]}}
 {id=Modulator_0, children [Attenuation_4]}
 {id=Modulator_5, children [WavePlayer_4], properties {pitchMax=1.1, pitchMin=0.9}}
 {
  id=SoundClass_2
  children [Attenuation_5]
  properties {soundClassOverride=&Fort_WeaponRanged_Fire_Distant}
 }
 {id=SoundClass_3, children [Modulator_0], properties {soundClassOverride=&Fort_WeaponRanged_Fire_3P_PSM}}
 {id=WavePlayer_0, properties {soundWaveAssetPtr=/Game/Sounds/Fort_GamePlay_Sounds/Explosives/Explosion_GrenadeProjectile_01_E}}
 {id=WavePlayer_4, properties {soundWaveAssetPtr=/Game/Sounds/Fort_GamePlay_Sounds/Explosives/explosion_sharp_distant_01}}
]
properties {
 maxDistance=15000
 volumeMultiplier=0.8
 concurrencyOverrides {maxCount=4, resolutionRule=StopOldest, volumeScale=0.5}
 concurrencySet [/Game/Sounds/Concurrency_Rules/Weapon_Explosion_Conc]
}
