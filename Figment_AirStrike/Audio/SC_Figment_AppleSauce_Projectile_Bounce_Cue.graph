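# Sound cue for the bounce of the AppleSauce projectile in Figment AirStrike, reusing smoke grenade bounce sounds.
# A Random node picks one of three equally weighted waves and feeds a Modulator that holds pitch fixed.
# A dedicated bounce attenuation and the ranged weapon projectile effects sound class apply.
# A concurrency override allows three simultaneous instances and prevents new ones beyond that.
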SoundCue /Figment_AirStrike/Audio/SC_Figment_AppleSauce_Projectile_Bounce_Cue
root=Modulator_0
nodes [
 {id=Modulator_0, children [Random_0], properties {pitchMax=1, pitchMin=1}}
 {id=Random_0, children [WavePlayer_6,WavePlayer_7,WavePlayer_8], properties {weights [1,1,1]}}
 {id=WavePlayer_6, properties {soundWaveAssetPtr=/Game/Athena/Sounds/Weapons/Smoke_Grenade/sfx_SmokeGrenade_Bounce_01}}
 {id=WavePlayer_7, properties {soundWaveAssetPtr=/Game/Athena/Sounds/Weapons/Smoke_Grenade/sfx_SmokeGrenade_Bounce_02}}
 {id=WavePlayer_8, properties {soundWaveAssetPtr=/Game/Athena/Sounds/Weapons/Smoke_Grenade/sfx_SmokeGrenade_Bounce_03}}
]
properties {
 attenuationSettings=&/ATT_Figment_AppleSauce_Bounce
 bOverrideConcurrency=true
 maxDistance=5500
 soundClassObject=&Fort_WeaponRanged_Projectile_Effects
 volumeMultiplier=0.7
 concurrencyOverrides {maxCount=3, resolutionRule=PreventNew}
}
