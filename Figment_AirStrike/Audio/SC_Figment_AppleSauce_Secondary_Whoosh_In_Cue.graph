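# Sound Cue for the incoming secondary whoosh of the Figment Air Strike weapon, codenamed AppleSauce.
# A Random node picks one of five equally weighted whoosh waves, then a Modulator randomizes its pitch upward.
# Overridden attenuation carries it over long range with low-pass muffling and reverb distance settings.
# Its own concurrency override caps simultaneous plays, blocking new ones and scaling volume as instances stack.
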SoundCue /Figment_AirStrike/Audio/SC_Figment_AppleSauce_Secondary_Whoosh_In_Cue
root=Modulator_0
nodes [
 {id=Modulator_0, children [Random_0], properties {pitchMax=1.5, pitchMin=1}}
 {id=Random_0, children [WavePlayer_0,WavePlayer_1,WavePlayer_2,WavePlayer_3,WavePlayer_6], properties {weights [1,1,1,1,1]}}
 {id=WavePlayer_0, properties {soundWaveAssetPtr=/Game/Athena/Sounds/Weapons/Air_Strike/AppleSauce_Secondary_Whoosh_In_01}}
 {id=WavePlayer_1, properties {soundWaveAssetPtr=/Game/Athena/Sounds/Weapons/Air_Strike/AppleSauce_Secondary_Whoosh_In_02}}
 {id=WavePlayer_2, properties {soundWaveAssetPtr=/Game/Athena/Sounds/Weapons/Air_Strike/AppleSauce_Secondary_Whoosh_In_03}}
 {id=WavePlayer_3, properties {soundWaveAssetPtr=/Game/Athena/Sounds/Weapons/Air_Strike/AppleSauce_Secondary_Whoosh_In_04}}
 {id=WavePlayer_6, properties {soundWaveAssetPtr=/Game/Athena/Sounds/Weapons/Air_Strike/AppleSauce_Secondary_Whoosh_In_05}}
]
properties {
 bOverrideAttenuation=true
 bOverrideConcurrency=true
 maxDistance=18000
 soundClassObject=&Fort_WeaponRanged_Fire_3P_PSM
 volumeMultiplier=0.8
 attenuationOverrides {
  bAttenuateWithLPF=true
  falloffDistance=15000
  lPFFrequencyAtMax=800
  lPFRadiusMax=10000
  lPFRadiusMin=5000
  nonSpatializedRadiusStart=800
  reverbDistanceMax=16000
  reverbDistanceMin=1000
  stereoSpread=1000
  attenuationShapeExtents {x=3000}
 }
 concurrencyOverrides {maxCount=5, resolutionRule=PreventNew, volumeScale=0.75, volumeScaleAttackTime=0.1}
}
