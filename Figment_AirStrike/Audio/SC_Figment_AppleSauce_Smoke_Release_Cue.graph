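# Sound cue for the smoke release sound in the Figment air strike weapon, routed to the explosives sound class.
# A single fuse smoke release wave plays through a modulator at reduced volume.
# Overridden attenuation fades it over distance and configures a send to a distant reverb submix.
# Overridden concurrency blocks new instances when the limit is reached instead of stopping playing ones.
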SoundCue /Figment_AirStrike/Audio/SC_Figment_AppleSauce_Smoke_Release_Cue
root=Modulator_1
nodes [
 {id=Modulator_1, children [WavePlayer_1]}
 {id=WavePlayer_1, properties {soundWaveAssetPtr=/Game/Athena/Sounds/Weapons/Air_Strike/AppleSauce_Fuse_Smoke_Release_01}}
]
properties {
 bOverrideAttenuation=true
 bOverrideConcurrency=true
 maxDistance=2512
 soundClassObject=/Game/Sounds/FortSoundClassesAndModes/Fort_Explosives
 volumeMultiplier=0.7
 attenuationOverrides {
  bEnableSubmixSends=true
  falloffDistance=2000
  attenuationShapeExtents {x=512}
  submixSendSettings [{
    bEnableHPFCutoff=false
    bEnableLPFCutoff=false
    disableManualSendClamp=false
    hPFCutoff=20
    lPFCutoff=20000
    maxSendDistance=6000
    maxSendLevel=1
    minSendDistance=400
    minSendLevel=0
    sendLevel=0.2
    sendLevelControlMethod=Linear
    soundSubmix=&Fortnite_Reverb_Distant
   }]
 }
 concurrencyOverrides {resolutionRule=PreventNew}
}
