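# Sound cue for the Figment AirStrike AppleSauce throw, judging by its name, playing a single player-run foley wave.
# One wave player feeds the output directly, with no randomization or modulation, at slightly reduced volume.
# It uses gadget-placement attenuation and the gadgets sound class, and overrides concurrency so new instances are prevented once the limit is reached.
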SoundCue /Figment_AirStrike/Audio/SC_Figment_AppleSauce_Throw_Cue
root=WavePlayer_28
nodes [{id=WavePlayer_28, properties {soundWaveAssetPtr=/Game/Sounds/Fort_Foley_Sounds/Moves/player_run_foley_11}}]
properties {
 attenuationSettings=/Game/Sounds/Attenuation/Fort_GadgetPlace_Att
 bOverrideConcurrency=true
 maxDistance=2000
 soundClassObject=&Fort_Gadgets
 volumeMultiplier=0.8
 concurrencyOverrides {resolutionRule=PreventNew}
}
