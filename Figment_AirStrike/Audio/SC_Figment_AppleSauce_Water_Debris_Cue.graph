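# Sound cue for water debris from explosions in the Figment AirStrike (AppleSauce) content, routed to the explosives sound class with overridden concurrency.
# A top mixer blends a close and a distant branch, each slightly delayed and given its own water-explosion attenuation.
# The close branch layers random wave and close-burst samples with pitch modulation.
# The distant branch plays a random distant burst at slightly lowered pitch, mixed a little quieter.
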SoundCue /Figment_AirStrike/Audio/SC_Figment_AppleSauce_Water_Debris_Cue
root=Mixer_5
nodes [
 {
  id=Attenuation_0
  children [Mixer_0]
  properties {
   attenuationSettings=/Game/Sounds/Attenuation/Fort_Explosion_Water_Close_Athena_Att
   attenuationOverrides {falloffDistance=900, attenuationShapeExtents {x=100}}
  }
 }
 {id=Attenuation_4, children [Modulator_3], properties {attenuationSettings=/Game/Sounds/Attenuation/Fort_Explosion_Water_Distant_Athena_Att}}
 {id=Delay_4, children [Attenuation_0], properties {delayMax=0.15, delayMin=0.15}}
 {id=Delay_5, children [Attenuation_4], properties {delayMax=0.15, delayMin=0.15}}
 {id=Mixer_0, children [Modulator_2,Modulator_4], properties {inputVolume [0.75,1]}}
 {id=Mixer_5, children [Delay_4,Delay_5], properties {inputVolume [1,0.7]}}
 {id=Modulator_2, children [Random_0], properties {pitchMax=1.2}}
 {id=Modulator_3, children [Random_3], properties {pitchMax=1, pitchMin=0.85, volumeMax=1, volumeMin=1}}
 {id=Modulator_4, children [Random_2], properties {pitchMax=1.2}}
 {
  id=Random_0
  children [WavePlayer_2,WavePlayer_3,WavePlayer_5,WavePlayer_6]
  properties {bShouldExcludeFromBranchCulling=true, weights [1,1,1,1]}
 }
 {
  id=Random_2
  children [WavePlayer_8, WavePlayer_9, WavePlayer_10, WavePlayer_11]
  properties {bShouldExcludeFromBranchCulling=true, weights [1,1,1,1]}
 }
 {id=Random_3, children [WavePlayer_12, WavePlayer_13, WavePlayer_14], properties {bShouldExcludeFromBranchCulling=true, weights [1,1,1]}}
 {id=WavePlayer_10, properties {soundWaveAssetPtr=/Game/Athena/Sounds/Weapons/Explosions/EXPL_Debris_Water_Burst_Close_03}}
 {id=WavePlayer_11, properties {soundWaveAssetPtr=/Game/Athena/Sounds/Weapons/Explosions/EXPL_Debris_Water_Burst_Close_04}}
 {id=WavePlayer_12, properties {soundWaveAssetPtr=/Game/Athena/Sounds/Weapons/Explosions/EXPL_Debris_Water_Burst_Distant_03}}
 {id=WavePlayer_13, properties {soundWaveAssetPtr=/Game/Athena/Sounds/Weapons/Explosions/EXPL_Debris_Water_Burst_Distant_02}}
 {id=WavePlayer_14, properties {soundWaveAssetPtr=/Game/Athena/Sounds/Weapons/Explosions/EXPL_Debris_Water_Burst_Distant_01}}
 {id=WavePlayer_2, properties {soundWaveAssetPtr=/Game/Athena/Sounds/Weapons/Explosions/EXPL_Debris_Water_Wave_04}}
 {id=WavePlayer_3, properties {soundWaveAssetPtr=/Game/Athena/Sounds/Weapons/Explosions/EXPL_Debris_Water_Wave_01}}
 {id=WavePlayer_5, properties {soundWaveAssetPtr=/Game/Athena/Sounds/Weapons/Explosions/EXPL_Debris_Water_Wave_02}}
 {id=WavePlayer_6, properties {soundWaveAssetPtr=/Game/Athena/Sounds/Weapons/Explosions/EXPL_Debris_Water_Wave_03}}
 {id=WavePlayer_8, properties {soundWaveAssetPtr=/Game/Athena/Sounds/Weapons/Explosions/EXPL_Debris_Water_Burst_Close_01}}
 {id=WavePlayer_9, properties {soundWaveAssetPtr=/Game/Athena/Sounds/Weapons/Explosions/EXPL_Debris_Water_Burst_Close_02}}
]
properties {
 bOverrideConcurrency=true
 maxDistance=10000
 soundClassObject=/Game/Sounds/FortSoundClassesAndModes/Fort_Explosives
 volumeMultiplier=0.7
 concurrencyOverrides {maxCount=3, volumeScale=0.5}
 concurrencySet [/Game/Sounds/Concurrency_Rules/Weapon_Explosion_Conc]
}
